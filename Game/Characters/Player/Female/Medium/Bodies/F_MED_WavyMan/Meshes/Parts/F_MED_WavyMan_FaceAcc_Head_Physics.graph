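# Physics asset for the WavyMan female character's head face-accessory mesh, simulating six pigtail bone chains.
# Capsules on each pigtail bone chain hang from a dyn_head capsule, which links to a head sphere and then a spine_05 capsule.
# Root links have tight swing and twist limits while lower links swing wider; the left B chain has an extra fourth segment.
# The head-to-spine joint leaves swing and twist free.
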PhysicsAsset /Game/Characters/Player/Female/Medium/Bodies/F_MED_WavyMan/Meshes/Parts/F_MED_WavyMan_FaceAcc_Head_Physics
bodies {
 dyn_head [{shapeType=Capsule, length=10.7256, radius=8.4129, center=<7.1866 0.9457 0>, rotation=<rot 90 0 -7.1092>}]
 head [{shapeType=Sphere, radius=8.3131, center=<0.0357 0.6849 -0.163>}]
 pigTailA_l_01 [{shapeType=Capsule, length=1.029, radius=1.8384, center=<-2.3983 0 0>, rotation=<rot 90 0 0>}]
 pigTailA_l_02 [{shapeType=Capsule, length=0.5988, radius=2.0797, center=<-2.4873 0.0001 0>, rotation=<rot 90 0 0>}]
 pigTailA_l_03 [{shapeType=Capsule, length=0.7852, radius=2.4196, center=<-2.7663 -0.0001 0>, rotation=<rot 90 0 0>}]
 pigTailA_r_01 [{shapeType=Capsule, length=2.0249, radius=1.8178, center=<2.567 0.0003 0>, rotation=<rot 90 0 0>}]
 pigTailA_r_02 [{shapeType=Capsule, length=0.758, radius=2.0385, center=<2.1606 0 0>, rotation=<rot 90 0 0>}]
 pigTailA_r_03 [{shapeType=Capsule, length=0.4845, radius=2.7161, center=<2.9717 -0.0028 -0.2122>, rotation=<rot 0 -90.0536 -85.9151>}]
 pigTailB_l_01 [{shapeType=Capsule, length=3.0317, radius=1.3775, center=<-2.8327 0.3386 -0.551>, rotation=<rot 90 14.0363 20.8494>}]
 pigTailB_l_02 [{shapeType=Capsule, length=1.1661, radius=1.6289, center=<-2.0317 0.1909 0>, rotation=<rot 90 0 0>}]
 pigTailB_l_03 [{shapeType=Capsule, length=2.2952, radius=1.9219, center=<-2.8591 0 0>, rotation=<rot 90 0 0>}]
 pigTailB_l_04 [{shapeType=Capsule, length=2.135, radius=2.2324, center=<-2.8731 -0.0001 0>, rotation=<rot 90 0 0>}]
 pigTailB_r_01 [{shapeType=Capsule, length=1.244, radius=1.9102, center=<2.6303 -0.0002 0>, rotation=<rot 90 0 0>}]
 pigTailB_r_02 [{shapeType=Capsule, length=1.2305, radius=2.0029, center=<2.7229 -0.0002 0>, rotation=<rot 90 0 0>}]
 pigTailB_r_03 [{shapeType=Capsule, length=0.2776, radius=2.7332, center=<3.1043 0 0>, rotation=<rot 90 0 0>}]
 pigTailC_l_01 [{shapeType=Capsule, length=2.0602, radius=1.9022, center=<-2.3901 0.4528 0.2287>, rotation=<rot 79.9897 62.6077 62.9652>}]
 pigTailC_l_02 [{shapeType=Capsule, length=2.4818, radius=2.0396, center=<-3.2116 -0.0001 0>, rotation=<rot 90 0 0>}]
 pigTailC_l_03 [{shapeType=Capsule, length=3.8479, radius=1.8969, center=<-3.7017 0 0>, rotation=<rot 90 0 0>}]
 pigTailC_r_01 [{shapeType=Capsule, length=0.4956, radius=1.8698, center=<2.0822 0 0>, rotation=<rot 90 0 0>}]
 pigTailC_r_02 [{shapeType=Capsule, length=1.4726, radius=2.0165, center=<2.6364 0.0002 0>, rotation=<rot 90 0 0>}]
 pigTailC_r_03 [{shapeType=Capsule, length=1.6332, radius=2.4051, center=<3.4283 0 0>, rotation=<rot 90 0 0>}]
 spine_05 [{shapeType=Capsule, length=7.5201, radius=8.2602, center=<8.0562 1.7263 -0.3365>, rotation=<rot 0 -168.6888 -87.3388>}]
}
constraints [
 {bone1Name=pigTailC_r_01, bone2Name=dyn_head, swing1LimitDegrees=8, swing2LimitDegrees=8, twistLimitDegrees=5}
 {bone1Name=pigTailC_r_02, bone2Name=pigTailC_r_01, swing1LimitDegrees=20, swing2LimitDegrees=20, twistLimitDegrees=8}
 {bone1Name=pigTailC_r_03, bone2Name=pigTailC_r_02, swing1LimitDegrees=20, swing2LimitDegrees=20, twistLimitDegrees=8}
 {bone1Name=pigTailB_r_01, bone2Name=dyn_head, swing1LimitDegrees=8, swing2LimitDegrees=8, twistLimitDegrees=5}
 {bone1Name=pigTailB_r_02, bone2Name=pigTailB_r_01, swing1LimitDegrees=20, swing2LimitDegrees=20, twistLimitDegrees=8}
 {bone1Name=pigTailB_r_03, bone2Name=pigTailB_r_02, swing1LimitDegrees=20, swing2LimitDegrees=20, twistLimitDegrees=8}
 {bone1Name=pigTailA_r_01, bone2Name=dyn_head, swing1LimitDegrees=8, swing2LimitDegrees=8, twistLimitDegrees=5}
 {bone1Name=pigTailA_r_02, bone2Name=pigTailA_r_01, swing1LimitDegrees=20, swing2LimitDegrees=20, twistLimitDegrees=8}
 {bone1Name=pigTailA_r_03, bone2Name=pigTailA_r_02, swing1LimitDegrees=20, swing2LimitDegrees=20, twistLimitDegrees=8}
 {bone1Name=dyn_head, bone2Name=head, swing1LimitDegrees=20, swing2LimitDegrees=20, twistLimitDegrees=5}
 {
  bone1Name=head
  bone2Name=spine_05
  swing1LimitDegrees=45
  swing1Motion=Free
  swing2LimitDegrees=45
  swing2Motion=Free
  twistLimitDegrees=45
  twistMotion=Free
 }
 {bone1Name=pigTailC_l_01, bone2Name=dyn_head, swing1LimitDegrees=8, swing2LimitDegrees=8, twistLimitDegrees=5}
 {bone1Name=pigTailC_l_02, bone2Name=pigTailC_l_01, swing1LimitDegrees=20, swing2LimitDegrees=20, twistLimitDegrees=8}
 {bone1Name=pigTailC_l_03, bone2Name=pigTailC_l_02, swing1LimitDegrees=20, swing2LimitDegrees=20, twistLimitDegrees=8}
 {bone1Name=pigTailB_l_01, bone2Name=dyn_head, swing1LimitDegrees=8, swing2LimitDegrees=8, twistLimitDegrees=5}
 {bone1Name=pigTailB_l_02, bone2Name=pigTailB_l_01, swing1LimitDegrees=20, swing2LimitDegrees=20, twistLimitDegrees=8}
 {bone1Name=pigTailB_l_03, bone2Name=pigTailB_l_02, swing1LimitDegrees=20, swing2LimitDegrees=20, twistLimitDegrees=8}
 {bone1Name=pigTailB_l_04, bone2Name=pigTailB_l_03, swing1LimitDegrees=20, swing2LimitDegrees=20, twistLimitDegrees=8}
 {bone1Name=pigTailA_l_01, bone2Name=dyn_head, swing1LimitDegrees=8, swing2LimitDegrees=8, twistLimitDegrees=5}
 {bone1Name=pigTailA_l_02, bone2Name=pigTailA_l_01, swing1LimitDegrees=20, swing2LimitDegrees=20, twistLimitDegrees=8}
 {bone1Name=pigTailA_l_03, bone2Name=pigTailA_l_02, swing1LimitDegrees=20, swing2LimitDegrees=20, twistLimitDegrees=8}
]
