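# Post-copy physics AnimBlueprint for the female medium WavyMan cosmetic's head face-accessory mesh, adding secondary motion.
# It copies pose and curves from the attached parent, runs RigidBody simulation based at spine_05, then chains AnimDynamics on both dyn_eye bones.
# While skydiving, Perlin noise sampled from pawn height drives an X/Y external force on the RigidBody, reduced when the parachute opens.
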AnimBlueprint /BRCosmetics/Characters/Player/Female/Medium/Bodies/F_MED_WavyMan/Meshes/Parts/F_MED_Wavy_Man_FaceAcc_Head_AnimBP
target_skeleton=/Game/Characters/Player/Female/Medium/Bodies/F_MED_WavyMan/Meshes/Parts/F_MED_WavyMan_FaceAcc_Head_Skeleton
graphs [
 AnimGraph {nodes [
   n1 = RigidBody(
    ComponentPose=Local To Component(LocalPose=Copy Pose From Mesh() {bCopyCurves=true, bUseAttachedParent=true})
    ExternalForce=Get falling_anim_dyn_up_force_vec3() -> (falling_anim_dyn_up_force_vec3:Vector)
   ) {
    bOverrideWorldGravity=true
    overridePhysicsAsset=/Game/Characters/Player/Female/Medium/Bodies/F_MED_WavyMan/Meshes/Parts/F_MED_WavyMan_FaceAcc_Head_Physics
    simulationSpace=BaseBoneSpace
    alphaBoolBlend {blendOption=Linear}
    baseBoneRef {boneName=spine_05}
    componentAppliedLinearAccClamp=<500 500 500>
    componentLinearAccScale=<0.25 0.25 0.25>
    componentLinearVelScale=<0.8 0.8 0.8>
    overrideWorldGravity {z=-980}
   }
   n2 = Anim Dynamics - Bone: dyn_eye_l(ComponentPose=@n1, Alpha=1, ExternalForce="0,0,0") {
    bGravityOverrideInSimSpace=true
    gravityScale=-1
    simulationSpace=Actor
    alphaBoolBlend {blendOption=Linear}
    boundBone {boneName=dyn_eye_l}
    componentAppliedLinearAccClamp=<5000 5000 5000>
    componentLinearAccScale=<0.3 0.3 0.3>
    componentLinearVelScale=<0.4 0.4 0.4>
    physicsBodyDefinitions [{
      sphereCollisionRadius=0
      boundBone {boneName=dyn_eye_l}
      boxExtents=<15 15 3.5>
      constraintSetup {angularLimitsMax=<30 30 30>, angularLimitsMin=<-30 -30 -30>, angularTarget=<1 0 0>}
      localJointOffset=<0 0 1>
     }]
    relativeSpaceBone {boneName=head}
   }
   n3 = Anim Dynamics - Bone: dyn_eye_r(ComponentPose=@n2, Alpha=1, ExternalForce="0,0,0") {
    bGravityOverrideInSimSpace=true
    gravityScale=-1
    simulationSpace=Actor
    alphaBoolBlend {blendOption=Linear}
    boundBone {boneName=dyn_eye_r}
    componentAppliedLinearAccClamp=<5000 5000 5000>
    componentLinearAccScale=<0.3 0.3 0.3>
    componentLinearVelScale=<0.4 0.4 0.4>
    physicsBodyDefinitions [{
      sphereCollisionRadius=0
      boundBone {boneName=dyn_eye_r}
      boxExtents=<10 10 2>
      constraintSetup {angularTargetAxis=AxisZ, twistAxis=AxisZ, angularLimitsMax=<30 30 30>, angularLimitsMin=<-30 -30 -30>, angularTarget=<0 0 1>}
      localJointOffset=<0 0 -1>
     }]
    relativeSpaceBone {boneName=head}
   }
   Output Pose(Result=Component To Local(ComponentPose=@n3)) {layerGroup=None}
  ]}
 EventGraph {nodes [
   Event BlueprintUpdateAnimation()
   n1 = Get PawnLastLocation() -> (PawnLastLocation:Vector)
   n2 = SelectFloat(
    A=Make Perlin Noise and Remap!(Value:float=BreakVector(InVec=@n1).Z, RangeOutMin:float=-15000, RangeOutMax:float=15000) -> (ReturnValue:float)
    B=0
    bPickA=Get bIsSkyDiving()
   )
   n3 = SelectFloat(A=float * float(A=@n2, B=0.2), B=@n2, bPickA=Get bIsParachuteOpened())
   > Set falling_anim_dyn_up_force_vec3(falling_anim_dyn_up_force_vec3:Vector=MakeVector(X=@n3, Y=@n3)) -> (Output_Get:Vector)
  ]}
]
variables [falling_anim_dyn_up_force_vec3 {type=Vector}]
